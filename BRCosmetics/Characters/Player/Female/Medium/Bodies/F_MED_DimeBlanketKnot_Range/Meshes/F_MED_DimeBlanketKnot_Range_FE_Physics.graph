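# Physics asset for the FE variant of the medium female DimeBlanketKnot_Range cosmetic body, built around a dynamic skirt with capsules on the lower spine, pelvis and legs.
# Eight skirt chains each run from a tiny base capsule, freely constrained to spine_02, through three dyn segments.
# Chain joints limit swing and mostly lock twist, except the front-middle chain's limited twist.
# Free-rotation cross-constraints tie some neighboring panels together at their middle and bottom segments.
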PhysicsAsset /BRCosmetics/Characters/Player/Female/Medium/Bodies/F_MED_DimeBlanketKnot_Range/Meshes/F_MED_DimeBlanketKnot_Range_FE_Physics
bodies {
 base_skirt_bk_mid [{shapeType=Capsule, length=1, radius=1, rotation=<rot -90 0 0>}]
 base_skirt_frnt_mid [{shapeType=Capsule, length=1, radius=1, rotation=<rot -90 0 0>}]
 base_skirt_side_bk_l [{shapeType=Capsule, length=1, radius=1, rotation=<rot -90 0 0>}]
 base_skirt_side_bk_r [{shapeType=Capsule, length=1, radius=1, rotation=<rot -90 0 0>}]
 base_skirt_side_frnt_l [{shapeType=Capsule, length=1, radius=1, rotation=<rot -90 0 0>}]
 base_skirt_side_frnt_r [{shapeType=Capsule, length=1, radius=1, rotation=<rot -90 0 0>}]
 base_skirt_side_l [{shapeType=Capsule, length=1, radius=1, rotation=<rot -90 0 0>}]
 base_skirt_side_r [{shapeType=Capsule, length=1, radius=1, rotation=<rot -90 0 0>}]
 calf_l [{shapeType=Capsule, length=30, radius=5.8471, shapeName=calf_l_capsule, center=<-20 -1.637 0>, rotation=<rot -90 0 0>}]
 calf_r [{shapeType=Capsule, length=30, radius=5.8469, shapeName=calf_r_capsule, center=<20 2.7327 0>, rotation=<rot -90 0 0>}]
 dyn_skirt_bk_mid_01 [{shapeType=Capsule, length=24, radius=5, center=<10 0 4.5>, rotation=<rot 0 0 90>}]
 dyn_skirt_bk_mid_02 [{shapeType=Capsule, length=24, radius=12, center=<10 0 10>, rotation=<rot 0 0 90>}]
 dyn_skirt_bk_mid_03 [{shapeType=Capsule, length=24, radius=9.5, center=<10 0 10>, rotation=<rot 0 0 90>}]
 dyn_skirt_frnt_mid_01 [{shapeType=Capsule, length=12, radius=5, center=<10 0 5>, rotation=<rot 0 0 90>}]
 dyn_skirt_frnt_mid_02 [{shapeType=Capsule, length=12, radius=12, center=<10 0 10>, rotation=<rot 0 0 90>}]
 dyn_skirt_frnt_mid_03 [{shapeType=Capsule, length=12, radius=12, center=<10 0 7>, rotation=<rot 0 0 90>}]
 dyn_skirt_side_bk_l_01 [{shapeType=Capsule, length=24, radius=5, center=<10 0 3.7>, rotation=<rot 0 0 90>}]
 dyn_skirt_side_bk_l_02 [{shapeType=Capsule, length=24, radius=7, center=<10 0 4.35>, rotation=<rot 0 0 90>}]
 dyn_skirt_side_bk_l_03 [{shapeType=Capsule, length=24, radius=7.5, center=<10 0 3>, rotation=<rot 0 0 90>}]
 dyn_skirt_side_bk_r_01 [{shapeType=Capsule, length=24, radius=5, center=<10 0 3.7>, rotation=<rot 0 0 90>}]
 dyn_skirt_side_bk_r_02 [{shapeType=Capsule, length=24, radius=7, center=<10 0 4.35>, rotation=<rot 0 0 90>}]
 dyn_skirt_side_bk_r_03 [{shapeType=Capsule, length=24, radius=7.5, center=<10 0 3>, rotation=<rot 0 0 90>}]
 dyn_skirt_side_frnt_l_01 [{shapeType=Capsule, length=12, radius=5, center=<10 0 5>, rotation=<rot 0 0 90>}]
 dyn_skirt_side_frnt_l_02 [{shapeType=Capsule, length=24, radius=12, center=<10 0 12>, rotation=<rot 0 0 90>}]
 dyn_skirt_side_frnt_l_03 [{shapeType=Capsule, length=24, radius=12, center=<10 0 10>, rotation=<rot 0 0 90>}]
 dyn_skirt_side_frnt_r_01 [{shapeType=Capsule, length=12, radius=5, center=<10 0 5>, rotation=<rot 0 0 90>}]
 dyn_skirt_side_frnt_r_02 [{shapeType=Capsule, length=12, radius=12, center=<10 0 12>, rotation=<rot 0 0 90>}]
 dyn_skirt_side_frnt_r_03 [{shapeType=Capsule, length=12, radius=12, center=<10 0 10>, rotation=<rot 0 0 90>}]
 dyn_skirt_side_l_01 [{shapeType=Capsule, length=12, radius=5, center=<10 0 3>, rotation=<rot 0 0 90>}]
 dyn_skirt_side_l_02 [{shapeType=Capsule, length=12, radius=5, center=<10 0 6>, rotation=<rot 0 0 90>}]
 dyn_skirt_side_l_03 [{shapeType=Capsule, length=24, radius=6.5, center=<10 0 8>, rotation=<rot 0 0 90>}]
 dyn_skirt_side_r_01 [{shapeType=Capsule, length=12, radius=5, center=<10 0 3>, rotation=<rot 0 0 90>}]
 dyn_skirt_side_r_02 [{shapeType=Capsule, length=12, radius=5, center=<10 0 6>, rotation=<rot 0 0 90>}]
 dyn_skirt_side_r_03 [{shapeType=Capsule, length=24, radius=6.5, center=<10 0 8>, rotation=<rot 0 0 90>}]
 foot_l [{shapeType=Capsule, length=12, radius=10, shapeName=foot_l_capsule, center=<0 5 0>, rotation=<rot 0 0 -90>}]
 foot_r [{shapeType=Capsule, length=10, radius=10, shapeName=foot_r_capsule, center=<0 -5 0>, rotation=<rot 0 0 -90>}]
 pelvis [
  {shapeType=Capsule, length=8, radius=11.92, shapeName=pelvis_capsule, center=<-1 -2.5713 0>}
  {shapeType=Capsule, length=8, radius=11.92, shapeName=pelvis_capsule, center=<-1 0 0>}
 ]
 spine_01 [{shapeType=Capsule, length=8, radius=10.6752, shapeName=spine_01_capsule, center=<-3.6273 -1.3601 0>}]
 spine_02 [{shapeType=Capsule, length=9.6647, radius=9.6647, shapeName=spine_02_capsule, center=<-2.6975 0 0>}]
 thigh_l [{shapeType=Capsule, length=20, radius=9.7511, shapeName=thigh_l_capsule, center=<-20 -1 0>, rotation=<rot -90 0 0>}]
 thigh_r [
  {shapeType=Capsule, length=20, radius=9.7511, shapeName=thigh_r_capsule, center=<20 1 0>, rotation=<rot -90 0 0>}
  {shapeType=Capsule, length=13, radius=9.7511, shapeName=thigh_r_capsule, center=<30.0259 0.9999 0>, rotation=<rot 0 -0.0007 0>}
 ]
}
constraints [
 {
  bone1Name=spine_02
  bone2Name=spine_01
  swing1LimitDegrees=45
  swing1Motion=Free
  swing2LimitDegrees=45
  swing2Motion=Free
  twistLimitDegrees=45
  twistMotion=Free
 }
 {
  bone1Name=spine_01
  bone2Name=pelvis
  swing1LimitDegrees=45
  swing1Motion=Free
  swing2LimitDegrees=45
  swing2Motion=Free
  twistLimitDegrees=45
  twistMotion=Free
 }
 {
  bone1Name=thigh_l
  bone2Name=pelvis
  swing1LimitDegrees=45
  swing1Motion=Free
  swing2LimitDegrees=45
  swing2Motion=Free
  twistLimitDegrees=45
  twistMotion=Free
 }
 {
  bone1Name=calf_l
  bone2Name=thigh_l
  swing1LimitDegrees=45
  swing1Motion=Free
  swing2LimitDegrees=45
  swing2Motion=Free
  twistLimitDegrees=45
  twistMotion=Free
 }
 {
  bone1Name=foot_l
  bone2Name=calf_l
  swing1LimitDegrees=45
  swing1Motion=Free
  swing2LimitDegrees=45
  swing2Motion=Free
  twistLimitDegrees=45
  twistMotion=Free
 }
 {
  bone1Name=foot_r
  bone2Name=calf_r
  swing1LimitDegrees=45
  swing1Motion=Free
  swing2LimitDegrees=45
  swing2Motion=Free
  twistLimitDegrees=45
  twistMotion=Free
 }
 {
  bone1Name=thigh_r
  bone2Name=pelvis
  swing1LimitDegrees=45
  swing1Motion=Free
  swing2LimitDegrees=45
  swing2Motion=Free
  twistLimitDegrees=45
  twistMotion=Free
 }
 {
  bone1Name=calf_r
  bone2Name=thigh_r
  swing1LimitDegrees=45
  swing1Motion=Free
  swing2LimitDegrees=45
  swing2Motion=Free
  twistLimitDegrees=45
  twistMotion=Free
 }
 {
  bone1Name=dyn_skirt_side_r_03
  bone2Name=dyn_skirt_side_r_02
  swing1LimitDegrees=15
  swing2LimitDegrees=15
  twistLimitDegrees=45
  twistMotion=Locked
 }
 {
  bone1Name=base_skirt_frnt_mid
  bone2Name=spine_02
  swing1LimitDegrees=45
  swing1Motion=Free
  swing2LimitDegrees=45
  swing2Motion=Free
  twistLimitDegrees=45
  twistMotion=Free
 }
 {bone1Name=dyn_skirt_frnt_mid_01, bone2Name=base_skirt_frnt_mid, swing1LimitDegrees=10, swing2LimitDegrees=45, twistLimitDegrees=5}
 {bone1Name=dyn_skirt_frnt_mid_02, bone2Name=dyn_skirt_frnt_mid_01, swing1LimitDegrees=10, swing2LimitDegrees=45, twistLimitDegrees=5}
 {bone1Name=dyn_skirt_frnt_mid_03, bone2Name=dyn_skirt_frnt_mid_02, swing1LimitDegrees=10, swing2LimitDegrees=20, twistLimitDegrees=5}
 {
  bone1Name=base_skirt_side_bk_l
  bone2Name=spine_02
  swing1LimitDegrees=45
  swing1Motion=Free
  swing2LimitDegrees=45
  swing2Motion=Free
  twistLimitDegrees=45
  twistMotion=Free
 }
 {
  bone1Name=dyn_skirt_side_bk_l_01
  bone2Name=base_skirt_side_bk_l
  swing1LimitDegrees=15
  swing2LimitDegrees=15
  twistLimitDegrees=15
  twistMotion=Locked
 }
 {
  bone1Name=dyn_skirt_side_bk_l_02
  bone2Name=dyn_skirt_side_bk_l_01
  swing1LimitDegrees=15
  swing2LimitDegrees=15
  twistLimitDegrees=45
  twistMotion=Locked
 }
 {
  bone1Name=dyn_skirt_side_bk_l_03
  bone2Name=dyn_skirt_side_bk_l_02
  swing1LimitDegrees=15
  swing2LimitDegrees=30
  twistLimitDegrees=45
  twistMotion=Locked
 }
 {
  bone1Name=base_skirt_side_bk_r
  bone2Name=spine_02
  swing1LimitDegrees=45
  swing1Motion=Free
  swing2LimitDegrees=45
  swing2Motion=Free
  twistLimitDegrees=45
  twistMotion=Free
 }
 {
  bone1Name=dyn_skirt_side_bk_r_01
  bone2Name=base_skirt_side_bk_r
  swing1LimitDegrees=15
  swing2LimitDegrees=15
  twistLimitDegrees=15
  twistMotion=Locked
 }
 {
  bone1Name=dyn_skirt_side_bk_r_02
  bone2Name=dyn_skirt_side_bk_r_01
  swing1LimitDegrees=15
  swing2LimitDegrees=15
  twistLimitDegrees=45
  twistMotion=Locked
 }
 {
  bone1Name=dyn_skirt_side_bk_r_03
  bone2Name=dyn_skirt_side_bk_r_02
  swing1LimitDegrees=15
  swing2LimitDegrees=30
  twistLimitDegrees=45
  twistMotion=Locked
 }
 {
  bone1Name=base_skirt_side_frnt_l
  bone2Name=spine_02
  swing1LimitDegrees=45
  swing1Motion=Free
  swing2LimitDegrees=45
  swing2Motion=Free
  twistLimitDegrees=45
  twistMotion=Free
 }
 {
  bone1Name=dyn_skirt_side_frnt_l_01
  bone2Name=base_skirt_side_frnt_l
  swing1LimitDegrees=15
  swing2LimitDegrees=45
  twistLimitDegrees=45
  twistMotion=Locked
 }
 {
  bone1Name=dyn_skirt_side_frnt_l_02
  bone2Name=dyn_skirt_side_frnt_l_01
  swing1LimitDegrees=15
  swing2LimitDegrees=45
  twistLimitDegrees=45
  twistMotion=Locked
 }
 {
  bone1Name=dyn_skirt_side_frnt_l_03
  bone2Name=dyn_skirt_side_frnt_l_02
  swing1LimitDegrees=15
  swing2LimitDegrees=45
  twistLimitDegrees=45
  twistMotion=Locked
 }
 {
  bone1Name=base_skirt_side_frnt_r
  bone2Name=spine_02
  swing1LimitDegrees=45
  swing1Motion=Free
  swing2LimitDegrees=45
  swing2Motion=Free
  twistLimitDegrees=45
  twistMotion=Free
 }
 {
  bone1Name=dyn_skirt_side_frnt_r_01
  bone2Name=base_skirt_side_frnt_r
  swing1LimitDegrees=15
  swing2LimitDegrees=10
  twistLimitDegrees=45
  twistMotion=Locked
 }
 {
  bone1Name=dyn_skirt_side_frnt_r_02
  bone2Name=dyn_skirt_side_frnt_r_01
  swing1LimitDegrees=15
  swing2LimitDegrees=10
  twistLimitDegrees=45
  twistMotion=Locked
 }
 {
  bone1Name=dyn_skirt_side_frnt_r_03
  bone2Name=dyn_skirt_side_frnt_r_02
  swing1LimitDegrees=15
  swing2LimitDegrees=20
  twistLimitDegrees=45
  twistMotion=Locked
 }
 {
  bone1Name=base_skirt_side_l
  bone2Name=spine_02
  swing1LimitDegrees=45
  swing1Motion=Free
  swing2LimitDegrees=45
  swing2Motion=Free
  twistLimitDegrees=45
  twistMotion=Free
 }
 {
  bone1Name=dyn_skirt_side_l_01
  bone2Name=base_skirt_side_l
  swing1LimitDegrees=15
  swing2LimitDegrees=15
  twistLimitDegrees=45
  twistMotion=Locked
 }
 {
  bone1Name=dyn_skirt_side_l_02
  bone2Name=dyn_skirt_side_l_01
  swing1LimitDegrees=15
  swing2LimitDegrees=15
  twistLimitDegrees=45
  twistMotion=Locked
 }
 {
  bone1Name=dyn_skirt_side_l_03
  bone2Name=dyn_skirt_side_l_02
  swing1LimitDegrees=15
  swing2LimitDegrees=15
  twistLimitDegrees=45
  twistMotion=Locked
 }
 {
  bone1Name=base_skirt_side_r
  bone2Name=spine_02
  swing1LimitDegrees=45
  swing1Motion=Free
  swing2LimitDegrees=45
  swing2Motion=Free
  twistLimitDegrees=45
  twistMotion=Free
 }
 {
  bone1Name=dyn_skirt_side_r_01
  bone2Name=base_skirt_side_r
  swing1LimitDegrees=15
  swing2LimitDegrees=15
  twistLimitDegrees=45
  twistMotion=Locked
 }
 {
  bone1Name=dyn_skirt_side_r_02
  bone2Name=dyn_skirt_side_r_01
  swing1LimitDegrees=15
  swing2LimitDegrees=15
  twistLimitDegrees=45
  twistMotion=Locked
 }
 {
  bone1Name=base_skirt_bk_mid
  bone2Name=spine_02
  swing1LimitDegrees=45
  swing1Motion=Free
  swing2LimitDegrees=45
  swing2Motion=Free
  twistLimitDegrees=45
  twistMotion=Free
 }
 {
  bone1Name=dyn_skirt_bk_mid_01
  bone2Name=base_skirt_bk_mid
  swing1LimitDegrees=15
  swing2LimitDegrees=15
  twistLimitDegrees=15
  twistMotion=Locked
 }
 {
  bone1Name=dyn_skirt_bk_mid_02
  bone2Name=dyn_skirt_bk_mid_01
  swing1LimitDegrees=15
  swing2LimitDegrees=20
  twistLimitDegrees=45
  twistMotion=Locked
 }
 {
  bone1Name=dyn_skirt_bk_mid_03
  bone2Name=dyn_skirt_bk_mid_02
  swing1LimitDegrees=15
  swing2LimitDegrees=30
  twistLimitDegrees=45
  twistMotion=Locked
 }
 {
  bone1Name=dyn_skirt_side_l_02
  bone2Name=dyn_skirt_side_frnt_l_02
  swing1LimitDegrees=45
  swing1Motion=Free
  swing2LimitDegrees=45
  swing2Motion=Free
  twistLimitDegrees=45
  twistMotion=Free
 }
 {
  bone1Name=dyn_skirt_side_l_03
  bone2Name=dyn_skirt_side_frnt_l_03
  swing1LimitDegrees=45
  swing1Motion=Free
  swing2LimitDegrees=45
  swing2Motion=Free
  twistLimitDegrees=45
  twistMotion=Free
 }
 {
  bone1Name=dyn_skirt_side_bk_r_02
  bone2Name=dyn_skirt_bk_mid_02
  swing1LimitDegrees=45
  swing1Motion=Free
  swing2LimitDegrees=45
  swing2Motion=Free
  twistLimitDegrees=45
  twistMotion=Free
 }
 {
  bone1Name=dyn_skirt_side_bk_l_02
  bone2Name=dyn_skirt_bk_mid_02
  swing1LimitDegrees=45
  swing1Motion=Free
  swing2LimitDegrees=45
  swing2Motion=Free
  twistLimitDegrees=45
  twistMotion=Free
 }
 {
  bone1Name=dyn_skirt_side_r_03
  bone2Name=dyn_skirt_side_frnt_r_03
  swing1LimitDegrees=45
  swing1Motion=Free
  swing2LimitDegrees=45
  swing2Motion=Free
  twistLimitDegrees=45
  twistMotion=Free
 }
 {
  bone1Name=dyn_skirt_side_r_02
  bone2Name=dyn_skirt_side_frnt_r_02
  swing1LimitDegrees=45
  swing1Motion=Free
  swing2LimitDegrees=45
  swing2Motion=Free
  twistLimitDegrees=45
  twistMotion=Free
 }
 {
  bone1Name=dyn_skirt_frnt_mid_02
  bone2Name=dyn_skirt_side_frnt_l_02
  swing1LimitDegrees=45
  swing1Motion=Free
  swing2LimitDegrees=45
  swing2Motion=Free
  twistLimitDegrees=45
  twistMotion=Free
 }
 {
  bone1Name=dyn_skirt_side_l_03
  bone2Name=dyn_skirt_side_bk_l_03
  swing1LimitDegrees=45
  swing1Motion=Free
  swing2LimitDegrees=45
  swing2Motion=Free
  twistLimitDegrees=45
  twistMotion=Free
 }
 {
  bone1Name=dyn_skirt_side_r_03
  bone2Name=dyn_skirt_side_bk_r_03
  swing1LimitDegrees=45
  swing1Motion=Free
  swing2LimitDegrees=45
  swing2Motion=Free
  twistLimitDegrees=45
  twistMotion=Free
 }
]
