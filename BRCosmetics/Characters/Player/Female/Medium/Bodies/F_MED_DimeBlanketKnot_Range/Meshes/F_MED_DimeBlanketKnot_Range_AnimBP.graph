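# Cosmetic body AnimBP for F_MED_DimeBlanketKnot_Range that animates its skirt and arm flaps on the pose copied from the attached parent.
# A Control Rig feeds separate gameplay and front-end RigidBody sims, with back-skirt constraints added for capes or backblings.
# Leg-angle, crouch and melee alphas then drive skirt correctives, and AnimDynamics swings the flaps except on ziplines.
# EventGraph computes these alphas, state-based gravity and a front-end idle curve that layers DefaultSlot over the arms.
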AnimBlueprint /BRCosmetics/Characters/Player/Female/Medium/Bodies/F_MED_DimeBlanketKnot_Range/Meshes/F_MED_DimeBlanketKnot_Range_AnimBP
target_skeleton=&/F_MED_DimeBlanketKnot_Range_Skeleton
graphs [
 AnimGraph {nodes [
   Local To Component()
   n1 = Get animD alpha() -> (animD alpha:double)
   n2 = "Constraint - dyn_skirt_bk_mid_01 (P)"(ComponentPose=Local To Component(LocalPose=Use cached pose 'RB_cache'()), Alpha=0.5, ConstraintWeights_0=1) {
    alphaBoolBlend {blendOption=Linear}
    boneToModify {boneName=dyn_skirt_bk_mid_01}
    constraintSetup [{targetBone {boneName=pelvis}}]
    constraintWeights [1]
   }
   n3 = "Constraint - dyn_skirt_side_bk_r_03 (R)"(ComponentPose=@n2, Alpha=0.5, ConstraintWeights_0=1) {
    alphaBoolBlend {blendOption=Linear}
    boneToModify {boneName=dyn_skirt_side_bk_r_03}
    constraintSetup [{transformType=Rotation, targetBone {boneName=thigh_r}}]
    constraintWeights [1]
   }
   n4 = "Constraint - dyn_skirt_side_bk_l_03 (R)"(ComponentPose=@n3, Alpha=0.5, ConstraintWeights_0=1) {
    alphaBoolBlend {blendOption=Linear}
    boneToModify {boneName=dyn_skirt_side_bk_l_03}
    constraintSetup [{transformType=Rotation, targetBone {boneName=thigh_l}}]
    constraintWeights [1]
   }
   n5 = "Constraint - dyn_skirt_side_bk_l_01 (P)"(ComponentPose=@n4, Alpha=0.5, ConstraintWeights_0=1) {
    alphaBoolBlend {blendOption=Linear}
    boneToModify {boneName=dyn_skirt_side_bk_l_01}
    constraintSetup [{targetBone {boneName=pelvis}}]
    constraintWeights [1]
   }
   n6 = "Constraint - dyn_skirt_side_bk_r_01 (P)"(ComponentPose=@n5, Alpha=0.5, ConstraintWeights_0=1) {
    alphaBoolBlend {blendOption=Linear}
    boneToModify {boneName=dyn_skirt_side_bk_r_01}
    constraintSetup [{targetBone {boneName=pelvis}}]
    constraintWeights [1]
   }
   n7 = Blend Poses by bool(
    BlendPose_0=Component To Local(ComponentPose=@n6)
    BlendPose_1=Use cached pose 'RB_cache'()
    bActiveValue=OR Boolean(A=Get Is Wearing Cape() -> (Is Wearing Cape:bool), B=Get isWearingBackbling())
   ) {blendType=Linear}
   n8 = RigidBody!(ComponentPose=Local To Component(LocalPose=Use cached pose 'CtrlRig_Cache'()), Alpha, OverrideWorldGravity:Vector="0,0,-980", ExternalForce) -> (Pose) {
    bClampLinearTranslationLimitToRefPose=true
    bFreezeIncomingPoseOnStart=true
    bOverrideWorldGravity=true
    overridePhysicsAsset=&/F_MED_DimeBlanketKnot_Range_FE_Physics
    simulationSpace=BaseBoneSpace
    alphaBoolBlend {blendOption=Linear}
    alphaScaleBiasClamp {interpSpeedDecreasing=20, interpSpeedIncreasing=0}
    baseBoneRef {boneName=pelvis}
    componentAppliedLinearAccClamp=<0 0 0>
    overrideWorldGravity {z=-980}
    simSpaceSettings {maxAngularAcceleration=100, maxAngularVelocity=4, maxLinearAcceleration=700, maxLinearVelocity=300, velocityScaleZ=0.5, worldAlpha=1}
   }
   n9 = Component To Local(ComponentPose=@n8)
   n10 = Get bIsInFrontEnd()
   n11 = "Constraint - dyn_skirt_frnt_mid_01 (R)"(
    ComponentPose=Local To Component(LocalPose=Blend Poses by bool(BlendPose_0=@n9, BlendPose_1=@n7, bActiveValue=@n10) {blendType=Linear})
    Alpha=Get leg_r_rot_crouch() -> (leg_r_rot_crouch:double)
    ConstraintWeights_0=1
   ) {
    alphaBoolBlend {blendOption=Linear}
    boneToModify {boneName=dyn_skirt_frnt_mid_01}
    constraintSetup [{transformType=Rotation, targetBone {boneName=thigh_r}}]
    constraintWeights [1]
   }
   n12 = "Transform (Modify) Bone - Bone: base_skirt_frnt_mid"(ComponentPose=@n11, Alpha=Get leg_l_rot_crouch() -> (leg_l_rot_crouch:double), Translation="0,0,0", Rotation="45,0,0", Scale="1,1,1") {
    rotationMode=BMM_Additive
    rotationSpace=BCS_BoneSpace
    scaleMode=BMM_Additive
    translationMode=BMM_Additive
    translationSpace=BCS_BoneSpace
    alphaBoolBlend {blendOption=Linear}
    boneToModify {boneName=base_skirt_frnt_mid}
    rotation {pitch=45}
   }
   n13 = Anim Dynamics - Bone: dyn_flap_arm_top_l(ComponentPose=@n12, Alpha=@n1, ExternalForce="0,0,0") {
    angularDampingOverride=1
    angularSpringConstant=20000
    bAngularSpring=true
    bOverrideAngularDamping=true
    bUseGravityOverride=true
    gravityScale=0.35
    simulationSpace=RootRelative
    alphaBoolBlend {blendOption=Linear}
    boundBone {boneName=dyn_flap_arm_top_l}
    componentLinearAccScale=<1 1 1>
    componentLinearVelScale=<1 1 1>
    gravityOverride {z=-980}
    physicsBodyDefinitions [{
      boundBone {boneName=dyn_flap_arm_top_l}
      boxExtents=<35 35 35>
      constraintSetup {angularLimitsMax=<15 15 15>, angularLimitsMin=<-15 -15 -15>, angularTarget=<1 0 0>}
      localJointOffset=<4 0 0>
     }]
    relativeSpaceBone {boneName=lowerarm_l}
   }
   n14 = Anim Dynamics - Bone: dyn_flap_arm_bottom_l(ComponentPose=@n13, Alpha=@n1, ExternalForce="0,0,0") {
    angularDampingOverride=1
    angularSpringConstant=20000
    bAngularSpring=true
    bOverrideAngularDamping=true
    bUseGravityOverride=true
    gravityScale=0.35
    simulationSpace=RootRelative
    alphaBoolBlend {blendOption=Linear}
    boundBone {boneName=dyn_flap_arm_bottom_l}
    componentLinearAccScale=<1 1 1>
    componentLinearVelScale=<1 1 1>
    gravityOverride {z=-980}
    physicsBodyDefinitions [{
      boundBone {boneName=dyn_flap_arm_bottom_l}
      boxExtents=<35 35 35>
      constraintSetup {angularLimitsMax=<15 15 15>, angularLimitsMin=<-15 -15 -15>, angularTarget=<1 0 0>}
      localJointOffset=<4 0 0>
     }]
    relativeSpaceBone {boneName=lowerarm_l}
   }
   n15 = Anim Dynamics - Bone: dyn_flap_arm_frnt_l(ComponentPose=@n14, Alpha=@n1, ExternalForce="0,0,0") {
    angularDampingOverride=1
    angularSpringConstant=20000
    bAngularSpring=true
    bOverrideAngularDamping=true
    bUseGravityOverride=true
    gravityScale=0.35
    simulationSpace=RootRelative
    alphaBoolBlend {blendOption=Linear}
    boundBone {boneName=dyn_flap_arm_frnt_l}
    componentLinearAccScale=<1 1 1>
    componentLinearVelScale=<1 1 1>
    gravityOverride {z=-980}
    physicsBodyDefinitions [{
      boundBone {boneName=dyn_flap_arm_frnt_l}
      boxExtents=<35 35 35>
      constraintSetup {angularLimitsMax=<15 15 15>, angularLimitsMin=<-15 -15 -15>, angularTarget=<1 0 0>}
      localJointOffset=<4 0 0>
     }]
    relativeSpaceBone {boneName=lowerarm_l}
   }
   n16 = Anim Dynamics - Bone: dyn_flap_arm_rear_l(ComponentPose=@n15, Alpha=@n1, ExternalForce="0,0,0") {
    angularDampingOverride=1
    angularSpringConstant=20000
    bAngularSpring=true
    bOverrideAngularDamping=true
    bUseGravityOverride=true
    gravityScale=0.35
    simulationSpace=RootRelative
    alphaBoolBlend {blendOption=Linear}
    boundBone {boneName=dyn_flap_arm_rear_l}
    componentLinearAccScale=<1 1 1>
    componentLinearVelScale=<1 1 1>
    gravityOverride {z=-980}
    physicsBodyDefinitions [{
      boundBone {boneName=dyn_flap_arm_rear_l}
      boxExtents=<35 35 35>
      constraintSetup {angularLimitsMax=<15 15 15>, angularLimitsMin=<-15 -15 -15>, angularTarget=<1 0 0>}
      localJointOffset=<4 0 0>
     }]
    relativeSpaceBone {boneName=lowerarm_l}
   }
   n17 = Anim Dynamics - Bone: dyn_flap_arm_top_r(ComponentPose=@n16, Alpha=@n1, ExternalForce="0,0,0") {
    angularDampingOverride=1
    angularSpringConstant=20000
    bAngularSpring=true
    bOverrideAngularDamping=true
    bUseGravityOverride=true
    gravityScale=0.35
    simulationSpace=BoneRelative
    alphaBoolBlend {blendOption=Linear}
    boundBone {boneName=dyn_flap_arm_top_r}
    componentLinearAccScale=<1 1 1>
    componentLinearVelScale=<1 1 1>
    gravityOverride {z=-980}
    physicsBodyDefinitions [{
      boundBone {boneName=dyn_flap_arm_top_r}
      boxExtents=<40 40 40>
      constraintSetup {angularLimitsMax=<15 15 15>, angularLimitsMin=<-15 -15 -15>, angularTarget=<1 0 0>}
      localJointOffset=<4 0 0>
     }]
    relativeSpaceBone {boneName=lowerarm_r}
   }
   n18 = Anim Dynamics - Bone: dyn_flap_arm_bottom_r(ComponentPose=@n17, Alpha=@n1, ExternalForce="0,0,0") {
    angularDampingOverride=1
    angularSpringConstant=20000
    bAngularSpring=true
    bOverrideAngularDamping=true
    bUseGravityOverride=true
    gravityScale=0.35
    simulationSpace=BoneRelative
    alphaBoolBlend {blendOption=Linear}
    boundBone {boneName=dyn_flap_arm_bottom_r}
    componentLinearAccScale=<1 1 1>
    componentLinearVelScale=<1 1 1>
    gravityOverride {z=-980}
    physicsBodyDefinitions [{
      boundBone {boneName=dyn_flap_arm_bottom_r}
      boxExtents=<40 40 40>
      constraintSetup {angularLimitsMax=<15 15 15>, angularLimitsMin=<-15 -15 -15>, angularTarget=<1 0 0>}
      localJointOffset=<4 0 0>
     }]
    relativeSpaceBone {boneName=lowerarm_r}
   }
   n19 = Anim Dynamics - Bone: dyn_flap_arm_frnt_r(ComponentPose=@n18, Alpha=@n1, ExternalForce="0,0,0") {
    angularDampingOverride=1
    angularSpringConstant=20000
    bAngularSpring=true
    bOverrideAngularDamping=true
    bUseGravityOverride=true
    gravityScale=0.35
    simulationSpace=BoneRelative
    alphaBoolBlend {blendOption=Linear}
    boundBone {boneName=dyn_flap_arm_frnt_r}
    componentLinearAccScale=<1 1 1>
    componentLinearVelScale=<1 1 1>
    gravityOverride {z=-980}
    physicsBodyDefinitions [{
      boundBone {boneName=dyn_flap_arm_frnt_r}
      boxExtents=<40 40 40>
      constraintSetup {angularLimitsMax=<15 15 15>, angularLimitsMin=<-15 -15 -15>, angularTarget=<1 0 0>}
      localJointOffset=<4 0 0>
     }]
    relativeSpaceBone {boneName=lowerarm_r}
   }
   n20 = Anim Dynamics - Bone: dyn_flap_arm_rear_r(ComponentPose=@n19, Alpha=@n1, ExternalForce="0,0,0") {
    angularDampingOverride=1
    angularSpringConstant=20000
    bAngularSpring=true
    bOverrideAngularDamping=true
    bUseGravityOverride=true
    gravityScale=0.35
    simulationSpace=BoneRelative
    alphaBoolBlend {blendOption=Linear}
    boundBone {boneName=dyn_flap_arm_rear_r}
    componentLinearAccScale=<1 1 1>
    componentLinearVelScale=<1 1 1>
    gravityOverride {z=-980}
    physicsBodyDefinitions [{
      boundBone {boneName=dyn_flap_arm_rear_r}
      boxExtents=<40 40 40>
      constraintSetup {angularLimitsMax=<15 15 15>, angularLimitsMin=<-15 -15 -15>, angularTarget=<1 0 0>}
      localJointOffset=<4 0 0>
     }]
    relativeSpaceBone {boneName=lowerarm_r}
   }
   n21 = "Transform (Modify) Bone - Bone: base_skirt_frnt_mid"(ComponentPose=@n20, Alpha=Get cross_leg_rot() -> (cross_leg_rot:double), Translation="0,0,0", Rotation="0,0,0", Scale="0.7,1,1") {
    rotationMode=BMM_Additive
    rotationSpace=BCS_BoneSpace
    scaleMode=BMM_Additive
    translationMode=BMM_Additive
    translationSpace=BCS_BoneSpace
    alphaBoolBlend {blendOption=Linear}
    boneToModify {boneName=base_skirt_frnt_mid}
    scale {x=0.7}
   }
   n22 = Get leg_l_rot_generic() -> (leg_l_rot_generic:double)
   n23 = "Transform (Modify) Bone - Bone: base_skirt_side_frnt_l"(ComponentPose=@n21, Alpha=@n22, Translation="0,0,0", Rotation="0,0,0", Scale="0.7,1,1") {
    rotationMode=BMM_Additive
    rotationSpace=BCS_BoneSpace
    scaleMode=BMM_Additive
    translationMode=BMM_Additive
    translationSpace=BCS_BoneSpace
    alphaBoolBlend {blendOption=Linear}
    boneToModify {boneName=base_skirt_side_frnt_l}
    scale {x=0.7}
   }
   n24 = "Transform (Modify) Bone - Bone: base_skirt_side_frnt_r"(ComponentPose=@n23, Alpha=Get leg_r_rot_generic() -> (leg_r_rot_generic:double), Translation="0,0,0", Rotation="0,0,0", Scale="0.7,1,1") {
    rotationMode=BMM_Additive
    rotationSpace=BCS_BoneSpace
    scaleMode=BMM_Additive
    scaleSpace=BCS_BoneSpace
    translationMode=BMM_Additive
    translationSpace=BCS_BoneSpace
    alphaBoolBlend {blendOption=Linear}
    boneToModify {boneName=base_skirt_side_frnt_r}
    scale {x=0.7}
   }
   n25 = "Transform (Modify) Bone - Bone: base_skirt_side_bk_l"(ComponentPose=@n24, Alpha=@n22, Translation="0,0,0", Rotation="5,0,0", Scale="1,1,1") {
    rotationMode=BMM_Additive
    rotationSpace=BCS_BoneSpace
    scaleMode=BMM_Additive
    scaleSpace=BCS_BoneSpace
    translationMode=BMM_Additive
    translationSpace=BCS_BoneSpace
    alphaBoolBlend {blendOption=Linear}
    boneToModify {boneName=base_skirt_side_bk_l}
    rotation {pitch=5}
   }
   n26 = Get pickAxe_float() -> (pickAxe_float:double)
   n27 = "Constraint - base_skirt_frnt_mid (RR)"!(ComponentPose=@n25, Alpha=@n26, ConstraintWeights_0=0.7, ConstraintWeights_1:float=0.3) -> (Pose) {
    alphaBoolBlend {blendInTime=0.2, blendOption=Linear}
    alphaScaleBiasClamp {bInterpResult=true}
    boneToModify {boneName=base_skirt_frnt_mid}
    constraintSetup [{transformType=Rotation, targetBone {boneName=thigh_l}}, {transformType=Rotation, targetBone {boneName=spine_01}}]
    constraintWeights [1,1]
   }
   n28 = "Constraint - base_skirt_side_frnt_l (RR)"!(ComponentPose=@n27, Alpha=@n26, ConstraintWeights_0=0.1, ConstraintWeights_1:float=0.9) -> (Pose) {
    alphaBoolBlend {blendInTime=0.2, blendOption=Linear}
    alphaScaleBiasClamp {bInterpResult=true}
    boneToModify {boneName=base_skirt_side_frnt_l}
    constraintSetup [{transformType=Rotation, targetBone {boneName=spine_01}}, {transformType=Rotation, targetBone {boneName=thigh_l}}]
    constraintWeights [1,1]
   }
   n29 = "Constraint - base_skirt_side_l (RR)"!(ComponentPose=@n28, Alpha=@n26, ConstraintWeights_0=1, ConstraintWeights_1:float=0) -> (Pose) {
    alphaBoolBlend {blendInTime=0.2, blendOption=Linear}
    alphaScaleBiasClamp {bInterpResult=true}
    boneToModify {boneName=base_skirt_side_l}
    constraintSetup [{transformType=Rotation, targetBone {boneName=thigh_l}}, {transformType=Rotation, targetBone {boneName=pelvis}}]
    constraintWeights [1,1]
   }
   n30 = Get crouchingMoving() -> (crouchingMoving:double)
   n31 = "Transform (Modify) Bone - Bone: base_skirt_side_frnt_r"(ComponentPose=@n29, Alpha=@n30, Translation="0,5,-5", Rotation="15,-20,0", Scale="1,1,1") {
    rotationMode=BMM_Additive
    rotationSpace=BCS_BoneSpace
    translationMode=BMM_Additive
    translationSpace=BCS_BoneSpace
    alphaBoolBlend {blendInTime=0.2, blendOption=Linear}
    alphaScaleBiasClamp {interpSpeedIncreasing=0}
    boneToModify {boneName=base_skirt_side_frnt_r}
    rotation {pitch=15, yaw=-20}
    translation {y=5, z=-5}
   }
   n32 = "Transform (Modify) Bone - Bone: dyn_skirt_frnt_mid_02"(ComponentPose=@n31, Alpha=@n30, Translation="0,0,0", Rotation="-30,0,0", Scale="1,1,1") {
    rotationMode=BMM_Additive
    rotationSpace=BCS_BoneSpace
    translationMode=BMM_Additive
    translationSpace=BCS_BoneSpace
    alphaBoolBlend {blendInTime=0.2, blendOption=Linear}
    alphaScaleBiasClamp {interpSpeedIncreasing=0}
    boneToModify {boneName=dyn_skirt_frnt_mid_02}
    rotation {pitch=-30}
   }
   n33 = "Constraint - dyn_skirt_side_frnt_r_02 (RR)"!(ComponentPose=@n32, bAlphaBoolEnabled:bool=True, ConstraintWeights_0=1, ConstraintWeights_1:float=0) -> (Pose) {
    alphaInputType=Bool
    alphaBoolBlend {blendInTime=0.2, blendOption=Linear, blendOutTime=0.2}
    alphaScaleBiasClamp {bInterpResult=true}
    boneToModify {boneName=dyn_skirt_side_frnt_r_02}
    constraintSetup [{transformType=Rotation, targetBone {boneName=thigh_r}}, {transformType=Rotation, targetBone {boneName=pelvis}}]
    constraintWeights [1,1]
   }
   n34 = Layered blend per bone(
    BasePose=Component To Local(ComponentPose=@n33)
    BlendPoses_0=Slot 'DefaultSlot'()
    BlendWeights_0=Get Custom Idle is Playing Float() -> (Custom Idle is Playing Float:double)
   ) {layerSetup [{branchFilters [
       {boneName=upperarm_r}
       {boneName=upperarm_twist_01_r}
       {boneName=upperarm_twist_02_r}
       {boneName=lowerarm_r}
       {boneName=lowerarm_twist_01_r}
       {boneName=lowerarm_twist_02_r}
       {boneName=hand_r}
       {boneName=upperarm_l}
       {boneName=upperarm_twist_01_l}
       {boneName=upperarm_twist_02_l}
       {boneName=lowerarm_l}
       {boneName=lowerarm_twist_01_l}
       {boneName=lowerarm_twist_02_l}
       {boneName=Nonehand_l}
      ]}]}
   Output Pose(Result=@n34) {layerGroup=None}
   n35 = RigidBody(ComponentPose=Local To Component(LocalPose=Use cached pose 'CtrlRig_Cache'()), OverrideWorldGravity=Get GravityOverride()) {
    bClampLinearTranslationLimitToRefPose=true
    bFreezeIncomingPoseOnStart=true
    bOverrideWorldGravity=true
    overridePhysicsAsset=&/F_MED_DimeBlanketKnot_Range_Physics
    simulationSpace=BaseBoneSpace
    alphaBoolBlend {blendOption=Linear}
    alphaScaleBiasClamp {interpSpeedDecreasing=20, interpSpeedIncreasing=0}
    baseBoneRef {boneName=pelvis}
    componentAppliedLinearAccClamp=<0 0 0>
    overrideWorldGravity {z=-980}
    simSpaceSettings {maxAngularAcceleration=100, maxAngularVelocity=4, maxLinearAcceleration=700, maxLinearVelocity=300, velocityScaleZ=0.5, worldAlpha=1}
   }
   n36 = "Transform (Modify) Bone - Bone: base_skirt_side_bk_r"!(ComponentPose=@n35, bAlphaBoolEnabled:bool=True, Translation="0,0,0", Rotation="5,0,0", Scale="1,1,1") -> (Pose) {
    alphaInputType=Bool
    rotationMode=BMM_Additive
    rotationSpace=BCS_BoneSpace
    alphaBoolBlend {blendInTime=0.2, blendOption=Linear, blendOutTime=0.2}
    boneToModify {boneName=base_skirt_side_bk_r}
    rotation {pitch=5}
   }
   n37 = "Transform (Modify) Bone - Bone: base_skirt_side_bk_l"!(ComponentPose=@n36, bAlphaBoolEnabled:bool=True, Translation="0,0,0", Rotation="5,0,0", Scale="1,1,1") -> (Pose) {
    alphaInputType=Bool
    rotationMode=BMM_Additive
    rotationSpace=BCS_BoneSpace
    alphaBoolBlend {blendInTime=0.2, blendOption=Linear, blendOutTime=0.2}
    boneToModify {boneName=base_skirt_side_bk_l}
    rotation {pitch=5}
   }
   n38 = "Transform (Modify) Bone - Bone: base_skirt_bk_mid"!(ComponentPose=@n37, bAlphaBoolEnabled:bool=True, Translation="0,0,0", Rotation="5,0,0", Scale="1,1,1") -> (Pose) {
    alphaInputType=Bool
    rotationMode=BMM_Additive
    rotationSpace=BCS_BoneSpace
    alphaBoolBlend {blendInTime=0.2, blendOption=Linear, blendOutTime=0.2}
    boneToModify {boneName=base_skirt_bk_mid}
    rotation {pitch=5}
   }
   n39 = "Transform (Modify) Bone - Bone: base_skirt_side_l"!(ComponentPose=@n38, bAlphaBoolEnabled:bool=True, Translation="0,0,0", Rotation="10,0,0", Scale="1,1,1") -> (Pose) {
    alphaInputType=Bool
    rotationMode=BMM_Additive
    rotationSpace=BCS_BoneSpace
    alphaBoolBlend {blendInTime=0.2, blendOption=Linear, blendOutTime=0.2}
    boneToModify {boneName=base_skirt_side_l}
    rotation {pitch=10}
   }
   n40 = "Transform (Modify) Bone - Bone: dyn_skirt_side_bk_r_02"!(ComponentPose=@n39, bAlphaBoolEnabled:bool=True, Translation="0,0,0", Rotation="-5,0,0", Scale="1,1,1") -> (Pose) {
    alphaInputType=Bool
    rotationMode=BMM_Additive
    rotationSpace=BCS_BoneSpace
    alphaBoolBlend {blendInTime=0.2, blendOption=Linear, blendOutTime=0.2}
    boneToModify {boneName=dyn_skirt_side_bk_r_02}
    rotation {pitch=-5}
   }
   n41 = "Transform (Modify) Bone - Bone: dyn_skirt_side_bk_l_02"!(ComponentPose=@n40, bAlphaBoolEnabled:bool=True, Translation="0,0,0", Rotation="-5,0,0", Scale="1,1,1") -> (Pose) {
    alphaInputType=Bool
    rotationMode=BMM_Additive
    rotationSpace=BCS_BoneSpace
    alphaBoolBlend {blendInTime=0.2, blendOption=Linear, blendOutTime=0.2}
    boneToModify {boneName=dyn_skirt_side_bk_l_02}
    rotation {pitch=-5}
   }
   n42 = "Transform (Modify) Bone - Bone: dyn_skirt_bk_mid_02"!(ComponentPose=@n41, bAlphaBoolEnabled:bool=True, Translation="0,0,0", Rotation="-5,0,0", Scale="1,1,1") -> (Pose) {
    alphaInputType=Bool
    rotationMode=BMM_Additive
    rotationSpace=BCS_BoneSpace
    alphaBoolBlend {blendInTime=0.2, blendOption=Linear, blendOutTime=0.2}
    boneToModify {boneName=dyn_skirt_bk_mid_02}
    rotation {pitch=-5}
   }
   n43 = "Transform (Modify) Bone - Bone: base_skirt_side_l"(ComponentPose=@n42, bAlphaBoolEnabled=True, Translation="0,0,0", Rotation="-15,0,0", Scale="1,1,1") {
    alphaInputType=Bool
    rotationMode=BMM_Additive
    rotationSpace=BCS_BoneSpace
    alphaBoolBlend {blendInTime=0.2, blendOption=Linear, blendOutTime=0.2}
    boneToModify {boneName=base_skirt_side_l}
    rotation {pitch=-15}
   }
   n44 = "Transform (Modify) Bone - Bone: base_skirt_side_r"!(ComponentPose=@n43, bAlphaBoolEnabled:bool=True, Translation="0,0,0", Rotation="-15,0,0", Scale="1,1,1") -> (Pose) {
    alphaInputType=Bool
    rotationMode=BMM_Additive
    rotationSpace=BCS_BoneSpace
    alphaBoolBlend {blendInTime=0.2, blendOption=Linear, blendOutTime=0.2}
    boneToModify {boneName=base_skirt_side_r}
    rotation {pitch=-15}
   }
   n45 = "Transform (Modify) Bone - Bone: base_skirt_side_frnt_l"!(ComponentPose=@n44, bAlphaBoolEnabled:bool=True, Translation="0,0,0", Rotation="30,0,0", Scale="1,1,1") -> (Pose) {
    alphaInputType=Bool
    rotationMode=BMM_Additive
    rotationSpace=BCS_BoneSpace
    alphaBoolBlend {blendInTime=0.2, blendOption=Linear, blendOutTime=0.2}
    boneToModify {boneName=base_skirt_side_frnt_l}
    rotation {pitch=30}
   }
   n46 = "Transform (Modify) Bone - Bone: base_skirt_side_frnt_r"!(ComponentPose=@n45, bAlphaBoolEnabled:bool=True, Translation="0,0,0", Rotation="30,0,0", Scale="1,1,1") -> (Pose) {
    alphaInputType=Bool
    rotationMode=BMM_Additive
    rotationSpace=BCS_BoneSpace
    alphaBoolBlend {blendInTime=0.2, blendOption=Linear, blendOutTime=0.2}
    boneToModify {boneName=base_skirt_side_frnt_r}
    rotation {pitch=30}
   }
   n47 = "Transform (Modify) Bone - Bone: base_skirt_frnt_mid"!(ComponentPose=@n46, bAlphaBoolEnabled:bool=True, Translation="0,0,0", Rotation="30,0,0", Scale="1,1,1") -> (Pose) {
    alphaInputType=Bool
    rotationMode=BMM_Additive
    rotationSpace=BCS_BoneSpace
    alphaBoolBlend {blendInTime=0.2, blendOption=Linear, blendOutTime=0.2}
    boneToModify {boneName=base_skirt_frnt_mid}
    rotation {pitch=30}
   }
   n48 = "Transform (Modify) Bone - Bone: base_skirt_side_l"(ComponentPose=@n47, bAlphaBoolEnabled=True, Translation="0,0,0", Rotation="15,0,0", Scale="1,1,1") {
    alphaInputType=Bool
    rotationMode=BMM_Additive
    rotationSpace=BCS_BoneSpace
    alphaBoolBlend {blendInTime=0.2, blendOption=Linear, blendOutTime=0.2}
    boneToModify {boneName=base_skirt_side_l}
    rotation {pitch=15}
   }
   Save cached pose 'RB_cache'(Pose=Component To Local(ComponentPose=@n48)) {cacheName=RB_cache}
   n49 = Control Rig!(
    Source=Copy Pose From Mesh() {bUseAttachedParent=true}
    Alpha
    isSkydiving:bool=Get bIsSkyDiving()
    isEmotePlaying:double=Get isEmotePlaying() -> (isEmotePlaying:double)
    isFrontEnd:bool=Get bIsInFrontEnd()
    isSliding:bool=Get bIsSliding()
   ) -> (Pose) {
    alphaBoolBlend {blendOption=Linear}
    controlRigAssetReference {blueprintRigClass=&/F_MED_DimeBlanketKnot_Range_CtrlRig.F_MED_DimeBlanketKnot_Range_CtrlRig_C}
   }
   Save cached pose 'CtrlRig_Cache'(Pose=@n49) {cacheName=CtrlRig_Cache}
   Reroute Node(InputPin:Wildcard) -> (OutputPin:Wildcard)
  ]}
 EventGraph {nodes [
   n1 = Event BlueprintUpdateAnimation()
   > Set deltaTime(deltaTime:double=@n1.DeltaTimeX) -> (Output_Get:double)
   > n2 = Sequence()
   n3 = AND Boolean(A=Get bIsCrouching(), B=Get bIsMoving2D())
   Set crouchingMoving(
    execute=@n2.then_1
    crouchingMoving:double=FInterpTo(Current=Get crouchingMoving(), Target="To Float (Boolean)"(InBool=@n3), DeltaTime=Get deltaTime() -> (deltaTime:double), InterpSpeed=2)
   ) -> (Output_Get:double)
   n4 = Get SkelMesh()
   n5 = GetAlphaFromReferenceTransformDelta(SkeletalMeshComponent=@n4, BoneName=pelvis, BaseName=thigh_l, InputRange=(X=0,Y=200), TwistAxis="0,0,1")
   > Set leg_l_rot_generic(leg_l_rot_generic:double=@n5) -> (Output_Get:double)
   n6 = Get crouchingMoving()
   > Set leg_l_rot_crouch(leg_l_rot_crouch:double=float * float(A=@n5, B="Clamp (Float)"(Value=@n6, Min=0.25, Max=1)) {is=title K2Node_PromotableOperator}) -> (Output_Get:double)
   n7 = GetAlphaFromReferenceTransformDelta(SkeletalMeshComponent=@n4, BoneName=pelvis, BaseName=thigh_r, InputRange=(X=0,Y=90), TwistAxis="0,0,1")
   > Set leg_r_rot_generic(leg_r_rot_generic:double=@n7) -> (Output_Get:double)
   n8 = Get crouchingMoving()
   > Set leg_r_rot_crouch(leg_r_rot_crouch:double=float * float(A=@n7, B="Clamp (Float)"(Value=@n8, Min=0.25, Max=0.75))) -> (Output_Get:double)
   > Set cross_leg_rot(cross_leg_rot:double=float * float(A=Get leg_l_rot_generic(), B=Get leg_r_rot_generic())) -> (Output_Get:double)
   n9 = Get bIsPlayingEmote()
   > Set isEmotePlaying(isEmotePlaying:double=FInterpTo(Current=Get isEmotePlaying(), Target="To Float (Boolean)"(InBool=@n9), DeltaTime=Get deltaTime(), InterpSpeed=2)) -> (Output_Get:double)
   n10 = NOT Boolean(A=Get bIsCrouching())
   n11 = Get bIsMoving2D()
   > Set isSkydivingg(isSkydivingg:bool=OR Boolean(A=AND Boolean(A=@n11, B=@n10), B=Get bIsSkyDiving())) -> (Output_Get:bool)
   > Set spine_rot(spine_rot:double=GetAlphaFromReferenceTransformDelta(SkeletalMeshComponent=Get SkelMesh(), BoneName=pelvis, BaseName=spine_05, InputRange=(X=0,Y=-200), TwistAxis="0,0,1")) -> (Output_Get:double)
   n12 = "To Float (Boolean)"(InBool=AND Boolean(A=Get bIsPlayingMeleeAnim(), B=Get bIsCrouching()))
   n13 = Get pickAxe_float()
   n14 = Get deltaTime()
   > Set pickAxe_float(pickAxe_float:double="Clamp (Float)"(Value=FInterpTo(Current=@n13, Target=@n12, DeltaTime=@n14, InterpSpeed=3), Min=0, Max=0.5)) -> (Output_Get:double)
   > Set isDBNOmoving(isDBNOmoving:bool=AND Boolean(A=Get bIsDBNO(), B=Get bIsMoving2D())) -> (Output_Get:bool)
   > Branch(Condition=Get bIsInFrontEnd())
   > Set Custom Idle is Playing Float(Custom Idle is Playing Float:double=GetCurveValue(self=GetOwnerFrontEndAnimInstance(), CurveName=Frontend_Idle_Bypass)) -> (Output_Get:double)
   n15 = Gravity Override State Function DEPRECATED(
    execute=@n2.then_0
    CharacterPartAnimInstance=Self-Reference()
    GameTimeInSeconds=GetGameTimeInSeconds()
    GravityOverrideParams=Get Gravity Override Params()
    RBAN Params=Get RBAN Params()
   )
   > Set GravityOverride(GravityOverride=@n15)
   n16 = Event OnCharacterCustomizationLoaded()
   > Set SkelMesh(SkelMesh=@n16.CustomBaseMesh)
   n17 = CheckForBackbling()
   > Set isWearingBackbling(isWearingBackbling=AND Boolean!(A=@n17, B=NOT Boolean(A=@n17.bIsBackblingACape), C:bool=NOT Boolean(A=@n17.bIsBackblingLong)) -> (ReturnValue))
   > Set Is Wearing Cape(Is Wearing Cape=OR Boolean(A=@n17.bIsBackblingACape, B=@n17.bIsBackblingLong))
   n18 = Event BlueprintProcessAnimInputEvent()
   > n19 = Branch(Condition=IsNamed(self=@n18.AnimInputEvent, InEventName=aie.Player.Zipline.Entered))
   > Set animD alpha(animD alpha:double=0) -> (Output_Get:double)
   Branch(execute=@n19.else, Condition=IsNamed(self=@n18.AnimInputEvent, InEventName=aie.Player.Zipline.Exited))
   > Set animD alpha(animD alpha=1)
   Event OnBeginCrouching()
   > Set isCrouchInterp(isCrouchInterp:double=1) -> (Output_Get:double)
   > Delay(Duration=0.3)
   > Set isCrouchInterp(isCrouchInterp=0)
   Get bIsDBNO()
  ]}
]
variables [
 animD alpha {type=number}
 cross_leg_rot {type=number}
 crouchingMoving {type=number}
 Custom Idle is Playing Float {type=number}
 deltaTime {type=number}
 Gravity Override Params {
  type=GravityOverrideParamsStruct
  default {emoteOrMelee {gravityOverride=<0 0 -700>}, onGround_Moving {gravityOverride=<0 0 -700>}, onGround_Standing {gravityOverride=<0 0 -700>}}
 }
 GravityOverride {type=Vector}
 Is Wearing Cape {type=boolean}
 isCrouchInterp {type=number}
 isDBNOmoving {type=boolean}
 isEmotePlaying {type=number}
 isPickaxeSwing {type=boolean}
 isSkydivingg {type=boolean}
 isWearingBackbling {type=boolean}
 leg_l_rot_crouch {type=number}
 leg_l_rot_generic {type=number}
 leg_r_rot_crouch {type=number}
 leg_r_rot_generic {type=number}
 OnZipline {type=boolean}
 pickAxe_float {type=number}
 RB_alpha {type=number}
 RBAN Params {
  type=RBANParamsStruct
  default {
   dBNO {simSpaceSettings {
     maxAngularAcceleration=1000
     maxAngularVelocity=3
     maxLinearAcceleration=300
     maxLinearVelocity=700
     velocityScaleZ=0
     externalLinearDragV=<0 0.5 0>
    }}
   dBNO_Carried {simSpaceSettings {
     maxAngularAcceleration=1000
     maxAngularVelocity=3
     maxLinearAcceleration=300
     maxLinearVelocity=700
     velocityScaleZ=0
     externalLinearDragV=<0 0.5 0>
    }}
   emoteOrMelee {simSpaceSettings {
     maxAngularAcceleration=100
     maxAngularVelocity=4
     maxLinearAcceleration=700
     maxLinearVelocity=300
     velocityScaleZ=0.5
     externalLinearDragV=<0 0 0>
    }}
   front_End {simSpaceSettings {
     maxAngularAcceleration=100
     maxAngularVelocity=4
     maxLinearAcceleration=700
     maxLinearVelocity=300
     velocityScaleZ=0.5
     externalLinearDragV=<0 0 0>
    }}
   onGround_Crouch_Moving {simSpaceSettings {
     maxAngularAcceleration=100
     maxAngularVelocity=4
     maxLinearAcceleration=700
     maxLinearVelocity=300
     velocityScaleZ=0.5
     externalLinearDragV=<0 0 0>
    }}
   onGround_Falling {simSpaceSettings {
     maxAngularAcceleration=1000
     maxAngularVelocity=3
     maxLinearAcceleration=300
     maxLinearVelocity=700
     velocityScaleZ=0
     externalLinearDragV=<0 0.5 0>
    }}
   onGround_Moving {simSpaceSettings {
     maxAngularAcceleration=100
     maxAngularVelocity=4
     maxLinearAcceleration=700
     maxLinearVelocity=300
     velocityScaleZ=0.5
     externalLinearDragV=<0 0 0>
    }}
   onGround_Standing {simSpaceSettings {
     maxAngularAcceleration=100
     maxAngularVelocity=4
     maxLinearAcceleration=700
     maxLinearVelocity=300
     velocityScaleZ=0.5
     externalLinearDragV=<0 0 0>
    }}
   rail_Grinding {simSpaceSettings {
     maxAngularAcceleration=1000
     maxAngularVelocity=3
     maxLinearAcceleration=300
     maxLinearVelocity=700
     velocityScaleZ=0
     worldAlpha=0.1
     externalLinearDragV=<0 0.5 0>
    }}
   riding_Moving {simSpaceSettings {
     maxAngularAcceleration=1000
     maxAngularVelocity=3
     maxLinearAcceleration=300
     maxLinearVelocity=700
     velocityScaleZ=0
     externalLinearDragV=<0 0.5 0>
    }}
   riding_Moving_Targeting {simSpaceSettings {
     maxAngularAcceleration=1000
     maxAngularVelocity=3
     maxLinearAcceleration=300
     maxLinearVelocity=700
     velocityScaleZ=0
     externalLinearDragV=<0 0.5 0>
    }}
   riding_Standing {simSpaceSettings {
     maxAngularAcceleration=1000
     maxAngularVelocity=3
     maxLinearAcceleration=300
     maxLinearVelocity=700
     velocityScaleZ=0
     externalLinearDragV=<0 0.5 0>
    }}
   skydive_DivingDown {simSpaceSettings {
     maxAngularAcceleration=100
     maxAngularVelocity=4
     maxLinearAcceleration=700
     maxLinearVelocity=300
     velocityScaleZ=0.5
     externalLinearDragV=<0 0 0>
    }}
   skydive_FloatingDown {
    componentLinearAccScale=<0 0.5 0>
    componentLinearVelScale=<0 0.5 0>
    simSpaceSettings {
     maxAngularAcceleration=100
     maxAngularVelocity=4
     maxLinearAcceleration=700
     maxLinearVelocity=300
     velocityScaleZ=0.5
     externalLinearDragV=<0 0 0>
    }
   }
   skydive_GoingUp {simSpaceSettings {
     maxAngularAcceleration=100
     maxAngularVelocity=4
     maxLinearAcceleration=700
     maxLinearVelocity=300
     velocityScaleZ=0.5
     externalLinearDragV=<0 0 0>
    }}
   skydive_Parachuting {
    componentLinearAccScale=<0 0.1 0.3>
    componentLinearVelScale=<0 0.1 0.3>
    simSpaceSettings {
     maxAngularAcceleration=100
     maxAngularVelocity=4
     maxLinearAcceleration=700
     maxLinearVelocity=300
     velocityScaleZ=0.5
     externalLinearDragV=<0 0 0>
    }
   }
   skydive_Surfing {simSpaceSettings {
     maxAngularAcceleration=1000
     maxAngularVelocity=3
     maxLinearAcceleration=300
     maxLinearVelocity=700
     velocityScaleZ=0
     externalLinearDragV=<0 0.5 0>
    }}
   swimming {simSpaceSettings {
     maxAngularAcceleration=100
     maxAngularVelocity=4
     maxLinearAcceleration=700
     maxLinearVelocity=300
     velocityScaleZ=0.5
     externalLinearDragV=<0 0 0>
    }}
   tactical_Sprint {simSpaceSettings {
     maxAngularAcceleration=300
     maxAngularVelocity=1
     maxLinearAcceleration=150
     maxLinearVelocity=150
     velocityScaleZ=0
     worldAlpha=0.25
     externalLinearDragV=<0 0 0>
    }}
   ziplining {simSpaceSettings {
     maxAngularAcceleration=1000
     maxAngularVelocity=3
     maxLinearAcceleration=300
     maxLinearVelocity=700
     velocityScaleZ=0
     externalLinearDragV=<0 0.5 0>
    }}
  }
 }
 SkelMesh {type=SkeletalMeshComponent}
 spine_rot {type=number}
]
